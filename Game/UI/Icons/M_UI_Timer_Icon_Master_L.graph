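# Translucent UI master material for an animated timer icon, with tintable border, tick and fill, plus an overall alpha.
# One packed texture supplies shape masks in R and G and the tick in B.
# Speed-scaled time drives an eased, looping off-center rotation and a sine-based scale pulse on the tick's UVs.
# Scalar lerp switches select animated or static tick and filled or outline-only icon, changing both color and opacity.
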Material /Game/UI/Icons/M_UI_Timer_Icon_Master_L
graph {
 properties {
  bUsedWithStaticMesh=false
  blendMode=BLEND_Translucent
  materialDomain=MD_UI
  pixelDepthOffsetMode=PDOM_Legacy
  usedShadingModels=Default Lit
 }
 nodes [
  n1 = VectorParameter() {parameterName=BorderTickColor, defaultValue=<rgba 1 1 1 1>}
  n2 = TextureSample(UVs=TextureCoordinate()) {texture=&/T_UI_Timer_128x}
  n3 = Multiply(A=Time(), B=ScalarParameter() {parameterName=TimerSpeedMult})
  n4 = Fmod(A=@n3, B=Constant() {r=1})
  n5 = Saturate(None=Power(Base=@n4, Exp=LinearInterpolate(Alpha=@n4) {constA=0.75, constB=0}) {constExponent=0.5})
  n6 = Add(A=@n3) {constB=-0.5}
  n7 = MaterialFunctionCall(UVs=TextureCoordinate(), Texture Scale=LinearInterpolate(Alpha=Sine(None=@n6)) {constA=0.9}) {materialFunction=&ScaleUVsByCenter}
  n8 = LinearInterpolate(
   A=@n2.B
   B=TextureSample(UVs=Rotator(Coordinate=@n7.UVs, Time=@n5) {centerX=0.495, centerY=0.575, speed=-6.2832}) {texture=&T_UI_Timer_128x}.B
   Alpha=ScalarParameter() {defaultValue=1, parameterName=UseRotation}
  )
  n9 = Add(A=LinearInterpolate(B=@n2.R, Alpha=OneMinus(None=@n2.G)), B=@n8)
  n10 = ScalarParameter() {defaultValue=1, parameterName=UseFill}
  n11 = OneMinus(None=@n2.G)
  LinearInterpolate(
   A=LinearInterpolate(A=@n1.RGB, B=@n1.RGB, Alpha=@n9)
   B=LinearInterpolate(A=VectorParameter() {parameterName=FillColor, defaultValue=<rgba 0.8308 0.016 0.1095 1>}.RGB, B=@n1.RGB, Alpha=Add(A=@n11, B=@n8))
   Alpha=@n10
  )
  Multiply(A=Saturate(None=LinearInterpolate(A=@n9, B=@n2.R, Alpha=@n10)), B=ScalarParameter() {defaultValue=1, parameterName=Overall_Alpha})
 ]
}
